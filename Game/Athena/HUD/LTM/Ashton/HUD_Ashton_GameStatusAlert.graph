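# Centered rich-text status alert widget for the Ashton LTM, showing game mode messages from the Athena game state.
# It handles only the three Center message types, playing sound and intro, then outro after the display time.
# Higher-priority messages cancel lower ones and clear the queue; equal or lower Important and Critical messages queue, while plain CenterMessages drop if busy.
# Outro completion shows the next queued message or clears the prioritized widget's active-content flag.
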WidgetBlueprint /Game/Athena/HUD/LTM/Ashton/HUD_Ashton_GameStatusAlert
animations [
 Intro {
  binding_count=1
  length_seconds=0.1709
  object=&self:Intro
  sequence {bindings [RichText_StatusText {
     id=#1
     possessed_class=RichTextBlock
     tracks [
      {
       type=MovieScene2DTransformTrack
       sections [{
         type=MovieScene2DTransformSection
         channels [
          Translation.X {type=MovieSceneScriptingFloatChannel}
          Translation.Y {type=MovieSceneScriptingFloatChannel}
          Angle {type=MovieSceneScriptingFloatChannel}
          Scale.X {type=MovieSceneScriptingFloatChannel, keys [[0,0.5],[2.0583,1],[3.4177,1]]}
          Scale.Y {type=MovieSceneScriptingFloatChannel, keys [[0,0.5],[2.0583,1],[3.4177,1]]}
          Shear.X {type=MovieSceneScriptingFloatChannel}
          Shear.Y {type=MovieSceneScriptingFloatChannel}
         ]
        }]
      }
      {
       type=MovieSceneFloatTrack
       display_name=Render Opacity
       sections [{type=MovieSceneFloatSection, channels [None {type=MovieSceneScriptingFloatChannel, keys [[0,0],[2.324,1]]}]}]
      }
     ]
    }]}
 }
 outro {
  binding_count=1
  length_seconds=0.5
  object=&self:outro
  sequence {bindings [RichText_StatusText {
     id=#2
     possessed_class=RichTextBlock
     tracks [
      {
       type=MovieScene2DTransformTrack
       sections [{
         type=MovieScene2DTransformSection
         end_frame=10
         start_frame=0
         channels [
          Translation.X {type=MovieSceneScriptingFloatChannel, const [0,1]}
          Translation.Y {type=MovieSceneScriptingFloatChannel, const [0,1]}
          Angle {type=MovieSceneScriptingFloatChannel, const [0,1]}
          Scale.X {type=MovieSceneScriptingFloatChannel, keys [[0,1],[10,0.5]]}
          Scale.Y {type=MovieSceneScriptingFloatChannel, keys [[0,1],[10,0.5]]}
          Shear.X {type=MovieSceneScriptingFloatChannel, const [-12,1]}
          Shear.Y {type=MovieSceneScriptingFloatChannel, const [0,1]}
         ]
        }]
      }
      {
       type=MovieSceneFloatTrack
       display_name=Render Opacity
       sections [{type=MovieSceneFloatSection, end_frame=10, start_frame=0, channels [None {type=MovieSceneScriptingFloatChannel, keys [[0,1],[10,0]]}]}]
      }
     ]
    }]}
 }
]
graphs [
 EventGraph {nodes [
   n1 = OnMessageRequested!() -> (OutputDelegate, then, MsgData:Athena Game Message Data Structure) {is=title K2Node_CustomEvent}
   > n2 = Switch on EAthenaGameMsgType!(
    execute
    Selection:EAthenaGameMsgType Enum=Break Athena Game Message Data!(AthenaGameMessageData:"Athena Game Message Data Structure (by ref)"=@n1.MsgData) -> (
     MsgType:EAthenaGameMsgType Enum
     MsgText:Text
     MsgSound:Sound Cue Object Reference
     MsgMediaSource:Media Source Object Reference
     MsgDelay:float=~
     bIsTeamBased:bool=~
     TeamIndex:int
     DisplayTime:float=~
     MessageFlags:EAthenaGameMessageDataFlags Enum
     TargetPlayerController:Player Controller Object Reference
    ) {is=title K2Node_BreakStruct}
   ) -> (
    None:Exec
    DefaultIntro:Exec
    DefaultMessage:Exec
    DefaultCriticalMessage:Exec
    CommIntro:Exec
    CommMessage:Exec
    CommCriticalMessage:Exec
    CornerIntro:Exec
    CornerMessage:Exec
    CornerCriticalMessage:Exec
    RespawnTurningOffWarning:Exec
    RespawnOffWarning:Exec
    CenterMessage:Exec
    CenterImportantMessage:Exec
    CenterCriticalMessage:Exec
   ) {is=title K2Node_SwitchEnum}
   n3 = Branch(execute=@n2.CenterCriticalMessage, Condition=Get bMsgDisplaying() -> (bMsgDisplaying:bool))
   > n4 = Switch on EAthenaGameMsgType(Selection=Break Athena Game Message Data(AthenaGameMessageData=Get CurrQueuedMsg() -> (CurrQueuedMsg:Athena Game Message Data Structure)) -> (MsgDelay=~, bIsTeamBased=~, DisplayTime=~))
   Add(
    execute=[@n11.then, @n5.then, @n10.CenterCriticalMessage, @n10.CenterImportantMessage, @n4.CenterCriticalMessage]
    TargetArray:Array of Athena Game Message Data Structures=Get QueuedMsgList() -> (QueuedMsgList:Array of Athena Game Message Data Structures)
    NewItem:"Athena Game Message Data Structure (by ref)"=@n1.MsgData
   ) {is=title K2Node_CallArrayFunction}
   n5 = CancelActiveMsg(execute=[@n4.CenterImportantMessage, @n4.CenterMessage])
   Set bMsgDisplaying(execute=[@n12.else, @n9.else, @n3.else], bMsgDisplaying:bool=true) -> (Output_Get:bool)
   > Set CurrQueuedMsg(CurrQueuedMsg:Athena Game Message Data Structure=@n1.MsgData) -> (Output_Get:Athena Game Message Data Structure)
   n6 = Break Athena Game Message Data(AthenaGameMessageData=@n1.MsgData) -> (MsgDelay=~, bIsTeamBased=~, DisplayTime=~)
   > SetText(self=Get RichText_StatusText() -> (RichText_StatusText:Rich Text Block Object Reference), InText=@n6.MsgText)
   > n7 = Is Valid(InputObject=@n6.MsgSound)
   SetHasActiveContent!(execute=[@n8.then, @n7.Is Not Valid], self:Fort Prioritized Widget Interface, bNewActive:bool=true) -> (then)
   > PlayAnimation(InAnimation=Get Intro())
   > Set Timer by Function Name!(
    execute
    Object:Object Reference
    FunctionName:String=OnDelayFinished
    Time:float=@n6.DisplayTime
    bLooping:bool=false
    bMaxOncePerFrame:bool=false
    InitialStartDelay:float=0
    InitialStartDelayVariance:float=0
   ) -> (then, ReturnValue:Timer Handle Structure) {is=title K2Node_CallFunction}
   n8 = PlaySound2D(execute=@n7.Is Valid, Sound=@n6.MsgSound)
   n9 = Branch(execute=@n2.CenterImportantMessage, Condition=Get bMsgDisplaying())
   > n10 = Switch on EAthenaGameMsgType(Selection=Break Athena Game Message Data(AthenaGameMessageData=Get CurrQueuedMsg()) -> (MsgDelay=~, bIsTeamBased=~, DisplayTime=~))
   n11 = CancelActiveMsg(execute=@n10.CenterMessage)
   n12 = Branch(execute=@n2.CenterMessage, Condition=Get bMsgDisplaying())
   OnOutroAnimFinished() {is=title K2Node_CustomEvent}
   > Set bMsgDisplaying(bMsgDisplaying=false)
   n13 = Get QueuedMsgList()
   > n14 = Branch(Condition="integer > integer"(A=Length(TargetArray:Array of Athena Game Message Data Structures=@n13), B=0) {is=title K2Node_CallFunction})
   > Set CurrQueuedMsg(CurrQueuedMsg="Get (a copy)"(Array:Array of Athena Game Message Data Structures=@n13) -> (Output:Athena Game Message Data Structure))
   > Remove Index!(execute, TargetArray:Array of Athena Game Message Data Structures=@n13, IndexToRemove:int=0) -> (then)
   > OnMessageRequested!(execute, self, MsgData:Athena Game Message Data Structure=Get CurrQueuedMsg()) -> (then) {is=title K2Node_CallFunction}
   SetHasActiveContent(execute=@n14.else, bNewActive=false)
   CancelActiveMsg() {is=title K2Node_CustomEvent}
   n15 = Get Intro()
   > n16 = Branch(Condition=IsAnimationPlaying(InAnimation=@n15))
   > n17 = StopAnimation(InAnimation=@n15)
   Clear Timer by Function Name!(execute=[@n17.then, @n16.else], Object:Object Reference, FunctionName:String=OnDelayFinished) -> (then)
   > Clear(TargetArray:Array of Athena Game Message Data Structures=Get QueuedMsgList())
   > n18 = Branch(Condition=IsAnimationPlaying(InAnimation=Get outro() -> (outro:Widget Animation Object Reference)))
   OnDelayFinished(execute=@n18.else)
   OnDelayFinished() {is=title K2Node_CustomEvent}
   > PlayAnimation(InAnimation=Get outro())
   Event Construct()
   > n19 = Cast To FortGameStateAthena!(execute, Object=GetGameState!() -> (ReturnValue:Game State Base Object Reference)) -> (then, CastFailed, AsFort Game State Athena:Fort Game State Athena Object Reference) {is=title K2Node_DynamicCast}
   > Bind Event to Game Mode Message Request(self:Fort Game State Athena Object Reference=@n19, Delegate=Create Event()) {is=title K2Node_AddDelegate}
   > BindToAnimationFinished(Animation=Get outro(), Delegate=Create Event())
   Event OnStompedByOtherWidget() {is=title K2Node_Event}
   n20 = Is Timer Active by Function Name!(Object:Object Reference, FunctionName:String=OnDelayFinished) -> (ReturnValue:bool)
   n21 = IsAnimationPlaying(InAnimation=Get Intro())
   > n22 = Branch(Condition=AND Boolean(A=OR Boolean(A=@n21, B=@n20), B=false))
   Branch(execute=@n22.else, Condition=IsAnimationPlaying(InAnimation=Get outro()))
   > StopAnimation(InAnimation=Get outro())
   > SetRenderOpacity(self=Get RichText_StatusText(), InOpacity=0)
   Branch()
   n23 = Get DEVPreviewString() -> (DEVPreviewString:int=0)
   > SetText(
    self=Get RichText_StatusText()
    InText=TextToUpper!(InText:"Text (by ref)"=Select!(
      Option 0:Text="LOCTEXT(\"<em>THE ENEMY</> HAS <em>10</> PLAYERS REMAINING!\")"
      Option 1:Text="LOCTEXT(\"THE ENEMY HAVE CAPTURED THE <Rock1>ROCK 1!</>\")"
      Option 2:Text="LOCTEXT(\"THE ENEMY HAVE CAPTURED THE <Rock2>ROCK 2!</>\")"
      Option 3:Text="LOCTEXT(\"THE ENEMY HAVE CAPTURED THE <Rock3>ROCK 3!</>\")"
      Option 4:Text="LOCTEXT(\"THE ENEMY HAVE CAPTURED THE <Rock4>ROCK 4!</>\")"
      Option 5:Text="LOCTEXT(\"THE ENEMY HAVE CAPTURED THE <Rock5>ROCK 5!</>\")"
      Option 6:Text="LOCTEXT(\"THE ENEMY HAVE CAPTURED THE <Rock6>ROCK 6!</>\")"
      Option 7:Text="LOCTEXT(\"<thanos>THANOS</> has done something to <hero>the heroes</>\\r\\nSo <em>do something</> before <thanos>Thanos</> it\\'s <infstone>too late!</>\\r\\nHe has the <rock1>Space stone!</>\\r\\nHe has the <rock2>Mind stone!</>\\r\\nHe has the <rock3>Power stone!</>\\r\\nHe has the <rock4>Time stone!</>\\r\\nHe has the <rock5>Soul stone!</>\\r\\nHe has the <rock6>Reality stone!</>\")"
      Option 8:Text="LOCTEXT(\"THE <rock1>Space stone!</> IS ARRIVING!\\r\\nGET THE LAST <em>INFINITY STONE</> TO GIVE THE GREAT <Thanos>THANOS</> THE VICTORY!\")"
      Index:int=@n23
     ) -> (ReturnValue:Text)) -> (ReturnValue:Text)
   )
   Event PreConstruct()
  ]}
 IntroDuration {nodes [
   Inputs!()
   n1 = Get Intro()
   Outputs!(ReturnValue:double=float - float(A=GetEndTime(self=@n1), B=GetStartTime!(self:Widget Animation Object Reference=@n1) -> (ReturnValue:float)) {is=title K2Node_CallFunction})
  ]}
]
variables [
 bMsgDisplaying {type=boolean}
 CurrQueuedMsg {
  type=AthenaGameMessageData
  default {bIsTeamBased=false, displayTime=5, messageFlags=None, msgDelay=0, msgMediaSource=null, msgSound=null, msgText="", msgType=None, teamIndex=0}
 }
 DEVPreviewString {type=integer, default=7}
 QueuedMsgList {type="array<AthenaGameMessageData>", default []}
]
widgets {
 info {inheritedWidgetCount=1, parentClass=/Script/FortniteUI.AthenaCommonPrioritizedWidget, rootWidgetClass=&Overlay, widgetCount=3}
 widgets [
  Overlay_0 {widgetClassPath=&Overlay, widget=Overlay_0}
  RichText_StatusText {
   widgetClassPath=/Script/UMG.RichTextBlock
   bIsVariable=true
   parent=Overlay_0
   widget=RichText_StatusText
   properties {
    justification=Center
    lineHeightPercentage=0.75
    textStyleSet=&/LTMAshtonTextStyles
    margin=<ltrb 0 -10 0 0>
    renderTransform {angle=0, scale=<1 1>, shear=<-12 0>, translation=<0 0>}
   }
   slot {class=OverlaySlot, properties {horizontalAlignment=HAlign_Center, verticalAlignment=VAlign_Center, padding=<ltrb 0 -10 0 -10>}}
  }
  TutorialHighlight {widgetClassPath=&TutorialHighlight}
 ]
}
